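# Niagara idle effect for the TaperSlink backpack cosmetic, adding socket-attached fire, sparks, smoke and a light that change between idle and running states.
# System timelines drive the small idle fire, larger running fire and flickering light, while an isRunning flag enables sparks and shuts off smoke spawning.
# Smoke sprites and a ribbon trail share a timeline-scaled fade.
# Particle partitions and spawn groups split single emitters into differently styled sprite layers.
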NiagaraSystem /BRCosmetics/Accessories/FORT_Backpacks/Backpack_M_LRG_TaperSlink/FX/NS_Backpack_TaperSlink_Idle
dependencies {
 usedDataInterfaces [
  &NiagaraDataInterfaceArrayFloat3
  &NiagaraDataInterfaceArrayFloat
  &NiagaraDataInterfaceCurve
  &NiagaraDataInterfaceVectorCurve
  &NiagaraDataInterfaceVector2DCurve
  &NiagaraDataInterfaceColorCurve
  &NiagaraDataInterfaceSpriteRendererInfo
  &NiagaraDataInterfaceSocketReader
 ]
 usedDynamicInputs [&VectorLength, &Multiply_Vector, &InvertBool]
 usedModules [
  &self:SystemSpawnScript.SystemScriptSource.SystemScriptGraph.NiagaraNodeAssignment_3.SetVariables_#1
  &SystemState
  &self:SystemSpawnScript.SystemScriptSource.SystemScriptGraph.NiagaraNodeAssignment_2.SetVariables_#2
  &Timeline
 ]
}
emitters {
 IdleFire {
  emitterUpdateScript [
   &EmitterState(Inactive Response=::NewEnumerator1, Life Cycle Mode=::NewEnumerator1, Loop Delay=0.27, Loop Duration=1, UseLoopDelay=false)
   &SpawnRate(SpawnRate=&Multiply_Float(A=50, B=$System.Timeline001.Float), Use Spawn Probability=false)
  ]
  particleSpawnScript [
   &PartitionParticles()
   /SetVariables(Particles.VisibilityTag=$Particles.PartitionParticles.Partition) {isSetParameters=true, module=SetVariables_#3}
   &InitializeParticle(
    Lifetime=&MakeCustomFloatFromBool(
     Bool=&SetBoolByFloatComparison(A=0.5, B=&MakeFloatFromInt(Int=$Particles.PartitionParticles.Partition))
     True Float=&RandomRangeFloat(Minimum=0.1, Maximum=0.2)
     False Float=&RandomRangeFloat(Minimum=0.025, Maximum=0.1)
    )
    Position Offset=<0 0 12>
    Sprite Rotation Mode=::NewEnumerator2
    Sprite Size Mode=::NewEnumerator3
    Uniform Sprite Size=&MakeCustomFloatFromBool(
     Bool=&SetBoolByFloatComparison(A=0.5, B=&MakeFloatFromInt(Int=$Particles.PartitionParticles.Partition))
     True Float=10
     False Float=&RandomRangeFloat(Minimum=2.5, Maximum=6)
    )
    UsePositionOffset=false
    Write Material Random=true
   )
   &Color(Color=&MakeLinearColorFromVectorAndFloat(
     "Vector (RGB)"=&SelectVectorFromArray(
      Direct Array Index=$Particles.PartitionParticles.Partition
      Vector Selection Array=DataInterface() {propertyValues {FloatData [<5 5 5>, <5 1.3858 0.7867>]}}
     )
     "Float (Alpha)"=&SelectFloatFromArray(Direct Array Index=$Particles.PartitionParticles.Partition, Float Selection Array=DataInterface() {propertyValues {FloatData [0.05,0.75]}})
    ))
  ]
  particleUpdateScript [
   &ParticleState()
   &SocketLocation(SamplePosOffset=false, Socket Reader=$System.Socket Reader, Socket Sampling Mode=::NewEnumerator3, VelocitySpeedLimit=false)
   &ScaleSpriteSize(
    Scale Sprite Size Mode=::NewEnumerator0
    Uniform Scale Factor=&MakeCustomFloatFromBool(
     Bool=&SetBoolByFloatComparison(A=0.5, B=&MakeFloatFromInt(Int=$Particles.PartitionParticles.Partition))
     True Float=1
     False Float=&FloatFromCurve(FloatCurve=DataInterface() {propertyValues {Curve {keys [{tangentMode=RCTM_User, time=0, value=0}, {tangentMode=RCTM_User, time=1, value=1}]}}})
    )
   )
   &ScaleColor(
    Scale Alpha=&FloatFromCurve(FloatCurve=DataInterface() {propertyValues {Curve {keys [
         {interpMode=RCIM_Cubic, tangentMode=RCTM_User, time=0, value=0}
         {interpMode=RCIM_Cubic, tangentMode=RCTM_User, time=0.5, value=1}
         {tangentMode=RCTM_User, time=1, value=0}
        ]}}})
    Scale RGB=&MakeCustomVectorFromBool(
     Bool=&SetBoolByFloatComparison(A=0.5, B=&MakeFloatFromInt(Int=$Particles.PartitionParticles.Partition))
     True Vector=&VectorFromCurve(
      VectorCurve=DataInterface() {propertyValues {
        XCurve {keys [<at 0 0.11>,<at 0.5 1>,<at 1 0.11>]}
        YCurve {keys [<at 0 0.05>,<at 0.5 0.13>,<at 1 0.05>]}
        ZCurve {keys [
          {arriveTangent=-0.3559, interpMode=RCIM_Cubic, leaveTangent=-0.3559, tangentMode=RCTM_User, time=0, value=0.03}
          {arriveTangent=0.6186, interpMode=RCIM_Cubic, leaveTangent=-0.8742, tangentMode=RCTM_Break, time=0.5, value=0}
          <at 1 0.03>
         ]}
       }}
      Scale Curve=<2 2 2>
     )
     False Vector=<1 1 1>
    )
    ScaleA=true
    ScaleRGB=true
   )
   &DynamicMaterialParameters(
    Index 0 Param 1=&FloatFromCurve(FloatCurve=DataInterface() {propertyValues {Curve {keys [<at 0 8>,<at 0.5 5>,<at 1 8>]}}})
    Index 0 Param 2=1
    Index 0 Param 3=1
    Index 0 Param 4=1
    Module.Index0Param0Write=true
    Module.Index0Param1Write=true
    Module.Index0Param2Write=true
    Module.Index0Param3Write=true
   )
   &CameraOffset(Camera Offset Amount=&SelectFloatFromArray(Direct Array Index=$Particles.PartitionParticles.Partition, Float Selection Array=DataInterface() {propertyValues {FloatData [5,2]}}))
   /Niagara/Modules/Update/SubUV/SubUVAnimation(Number Of Frames=&SelectFloatFromArray(Direct Array Index=$Particles.PartitionParticles.Partition, Float Selection Array=DataInterface() {propertyValues {FloatData [0,3]}}))
   /SetVariables(Particles.GlowColor=&MultiplyLinearColor(Linear Color=$Particles.Color, Scale Factor=<1 1 1 0.7>)) {isSetParameters=true}
  ]
  rendererClasses [&NiagaraSpriteRendererProperties]
  renderers [
   {rendererClass=&NiagaraSpriteRendererProperties}
   {rendererClass=&NiagaraSpriteRendererProperties, rendererIndex=1}
  ]
 }
 Light {
  emitterUpdateScript [
   &EmitterState(Life Cycle Mode=::NewEnumerator1, Loop Behavior=::NewEnumerator1, Loop Duration=1, UseLoopDelay=false)
   &SpawnBurst_Instantaneous(Use Spawn Probability=false, UseLoopCountLimit=false)
  ]
  particleSpawnScript [&InitializeParticle(Color Mode=::NewEnumerator0, Lifetime=5, UsePositionOffset=false, Write Material Random=true)]
  particleUpdateScript [
   &ParticleState(Kill Particles When Lifetime Has Elapsed=false)
   &SocketLocation(SamplePosOffset=false, Socket Reader=$System.Socket Reader, Socket Sampling Mode=::NewEnumerator3, VelocitySpeedLimit=false)
   &FN_Light_Attributes(
    Color=&MultiplyLinearColor(
     Linear Color=&ColorFromCurve(
      ColorCurve=DataInterface() {propertyValues {
        AlphaCurve {keys [<at 0 1 RCIM_Cubic>]}
        BlueCurve {keys [<at 0 0 RCIM_Cubic>]}
        GreenCurve {keys [<at 0 3.6563 RCIM_Cubic>]}
        RedCurve {keys [<at 0 12 RCIM_Cubic>]}
       }}
      CurveIndex=$Emitter.NormalizedLoopAge
     )
     Scale Factor=&MakeVector4_FromFloat(Float=&FloatFromCurve(
       FloatCurve=DataInterface() {propertyValues {Curve {keys [{tangentMode=RCTM_User, time=0.5, value=0}, {tangentMode=RCTM_User, time=1, value=1}]}}}
       CurveIndex=$Emitter.Age
      ))
    )
    Create Color Attribute=true
    Create Exponent Attribute=true
    Create Radius Attribute=true
    Enable Visibility Tag=true
    Radius=30
    Scale Color=&Waveform(
     Global Amplitude Scale=&Multiply_Float(
      A=&Lerp_Float(A=5, B=10, Alpha=$System.Timeline.Time)
      B=&OneMinusFloat(Float=$System.Timeline002.Float)
     )
     "Waveform [1]"=::NewEnumerator0
     "[1] Frequency"=5
     "[1] Amplitude Min/Max"=<0.75 1>
    )
   )
  ]
  rendererClasses [&NiagaraLightRendererProperties]
  renderers [{rendererClass=&NiagaraLightRendererProperties, rendererIndex=0}]
 }
 RunningFire {
  emitterUpdateScript [
   &EmitterState(Inactive Response=::NewEnumerator1, Life Cycle Mode=::NewEnumerator1, Loop Delay=0.27, Loop Duration=1, UseLoopDelay=false)
   &SpawnRate(SpawnRate=&Multiply_Float(A=50, B=$System.Timeline.Float), Use Spawn Probability=false)
  ]
  particleSpawnScript [
   &PartitionParticles()
   /SetVariables(Particles.VisibilityTag=$Particles.PartitionParticles.Partition) {isSetParameters=true, module=SetVariables_#3}
   &InitializeParticle(
    Lifetime=&MakeCustomFloatFromBool(
     Bool=&SetBoolByFloatComparison(A=0.5, B=&MakeFloatFromInt(Int=$Particles.PartitionParticles.Partition))
     True Float=&RandomRangeFloat(Minimum=0.1, Maximum=0.2)
     False Float=&RandomRangeFloat(Minimum=0.025, Maximum=0.1)
    )
    Position Offset=<0 0 12>
    Sprite Rotation Mode=::NewEnumerator2
    Sprite Size Mode=::NewEnumerator3
    Uniform Sprite Size=&MakeCustomFloatFromBool(
     Bool=&SetBoolByFloatComparison(A=0.5, B=&MakeFloatFromInt(Int=$Particles.PartitionParticles.Partition))
     True Float=20
     False Float=&RandomRangeFloat(Minimum=5, Maximum=12)
    )
    UsePositionOffset=false
    Write Material Random=true
   )
   &Color(Color=&MakeLinearColorFromVectorAndFloat(
     "Vector (RGB)"=&SelectVectorFromArray(
      Direct Array Index=$Particles.PartitionParticles.Partition
      Vector Selection Array=DataInterface() {propertyValues {FloatData [<5 5 5>, <5 1.3858 0.7867>]}}
     )
     "Float (Alpha)"=&SelectFloatFromArray(Direct Array Index=$Particles.PartitionParticles.Partition, Float Selection Array=DataInterface() {propertyValues {FloatData [0.1,1]}})
    ))
  ]
  particleUpdateScript [
   &ParticleState()
   &SocketLocation(SamplePosOffset=false, Socket Reader=$System.Socket Reader, Socket Sampling Mode=::NewEnumerator3, VelocitySpeedLimit=false)
   &ScaleSpriteSize(
    Scale Sprite Size Mode=::NewEnumerator0
    Uniform Scale Factor=&MakeCustomFloatFromBool(
     Bool=&SetBoolByFloatComparison(A=0.5, B=&MakeFloatFromInt(Int=$Particles.PartitionParticles.Partition))
     True Float=1
     False Float=&FloatFromCurve(FloatCurve=DataInterface() {propertyValues {Curve {keys [{tangentMode=RCTM_User, time=0, value=0}, {tangentMode=RCTM_User, time=1, value=1}]}}})
    )
   )
   &ScaleColor(
    Scale Alpha=&FloatFromCurve(FloatCurve=DataInterface() {propertyValues {Curve {keys [
         {interpMode=RCIM_Cubic, tangentMode=RCTM_User, time=0, value=0}
         {interpMode=RCIM_Cubic, tangentMode=RCTM_User, time=0.5, value=1}
         {tangentMode=RCTM_User, time=1, value=0}
        ]}}})
    Scale RGB=&MakeCustomVectorFromBool(
     Bool=&SetBoolByFloatComparison(A=0.5, B=&MakeFloatFromInt(Int=$Particles.PartitionParticles.Partition))
     True Vector=&VectorFromCurve(
      VectorCurve=DataInterface() {propertyValues {
        XCurve {keys [<at 0 0.11>,<at 0.5 1>,<at 1 0.11>]}
        YCurve {keys [<at 0 0.05>,<at 0.5 0.13>,<at 1 0.05>]}
        ZCurve {keys [
          {arriveTangent=-0.3559, interpMode=RCIM_Cubic, leaveTangent=-0.3559, tangentMode=RCTM_User, time=0, value=0.03}
          {arriveTangent=0.6186, interpMode=RCIM_Cubic, leaveTangent=-0.8742, tangentMode=RCTM_Break, time=0.5, value=0}
          <at 1 0.03>
         ]}
       }}
      Scale Curve=<2 2 2>
     )
     False Vector=<1 1 1>
    )
    ScaleA=true
    ScaleRGB=true
   )
   &DynamicMaterialParameters(
    Index 0 Param 1=&FloatFromCurve(FloatCurve=DataInterface() {propertyValues {Curve {keys [<at 0 8>,<at 0.5 5>,<at 1 8>]}}})
    Index 0 Param 2=1
    Index 0 Param 3=1
    Index 0 Param 4=1
    Module.Index0Param0Write=true
    Module.Index0Param1Write=true
    Module.Index0Param2Write=true
    Module.Index0Param3Write=true
   )
   &CameraOffset(Camera Offset Amount=&SelectFloatFromArray(Direct Array Index=$Particles.PartitionParticles.Partition, Float Selection Array=DataInterface() {propertyValues {FloatData [5,2]}}))
   &SubUVAnimation(Number Of Frames=&SelectFloatFromArray(Direct Array Index=$Particles.PartitionParticles.Partition, Float Selection Array=DataInterface() {propertyValues {FloatData [0,3]}}))
   /SetVariables(Particles.GlowColor=&MultiplyLinearColor(Linear Color=$Particles.Color, Scale Factor=<1 1 1 0.65>)) {isSetParameters=true}
  ]
  rendererClasses [&NiagaraSpriteRendererProperties]
  renderers [
   {rendererClass=&NiagaraSpriteRendererProperties}
   {rendererClass=&NiagaraSpriteRendererProperties, rendererIndex=1}
  ]
 }
 RunningSparks {
  emitterUpdateScript [
   &EmitterState(Life Cycle Mode=::NewEnumerator1, Loop Delay=0.2, Loop Duration=1, UseLoopDelay=false)
   &SpawnRate(
    SpawnRate=&FloatFromCurve(
     FloatCurve=DataInterface() {propertyValues {Curve {keys [<at 0.75 75 RCIM_Cubic>, {arriveTangent=-14.2993, leaveTangent=-14.2993, time=1, value=0}]}}}
     CurveIndex=$Emitter.NormalizedLoopAge
     Scale Curve=&MakeCustomFloatFromBool(Bool=$System.isRunning, True Float=1, False Float=0)
    )
    Use Spawn Probability=false
   )
   &SpawnRate(
    SpawnGroup=1
    SpawnRate=&MakeCustomFloatFromBool(Bool=$System.isRunning, True Float=50, False Float=0)
    Use Spawn Probability=false
   ) {module=SpawnRate001}
  ]
  particleSpawnScript [
   /SetVariables(Particles.VisibilityTag=$Emitter.SpawnGroup) {isSetParameters=true}
   &InitializeParticle(
    Lifetime=&MaskFloatBySpawnGroup(
     Masked Float=&RandomRangeFloat(Minimum=0.1, Maximum=0.5)
     Passthrough Float=&RandomRangeFloat(Minimum=0.025, Maximum=0.1)
    )
    Position Offset=<0 0 50>
    Sprite Size=&Multiply_Vector2D_ByFloat(
     Float=0.5
     Vector2D=&MaskVector2dBySpawnGroup(
      Masked Vector2D=&RandomRangeVector2D(Minimum=<1 1.5>, Maximum=<1.5 1.5>)
      Passthrough Vector2D=&Vector2DFromFloat(Value=&RandomRangeFloat(Minimum=2, Maximum=10))
      Spawn Group Mask=0
     )
    )
    Sprite Size Mode=::NewEnumerator4
    UsePositionOffset=false
    Write Material Random=true
   )
   &SocketLocation(SamplePosOffset=false, Socket Reader=$System.Socket Reader, Socket Sampling Mode=::NewEnumerator3, VelocitySpeedLimit=false)
   &ShapeLocation(
    MaskSpawnGroup=false
    Override Seed=false
    Shape Origin=$Output.SocketLocation.SampledPosition
    Sphere Radius=&MaskFloatBySpawnGroup(Masked Float=0, Passthrough Float=&RandomRangeFloat(Minimum=0.5, Maximum=1))
    UseOwnerScale=true
   )
   &AddVelocity(
    Velocity=&MaskVectorBySpawnGroup(
     Masked Vector=&RandomRangeVector(Minimum=<-250 -250 -250>, Maximum=<250 250 250>)
     Passthrough Vector=&Multiply_VectorByFloat(Vector=$Particles.ShapeLocation.ShapeVector, Float=5)
    )
    Velocity Speed Scale=0.25
   )
   &Color(Color=&MakeLinearColorFromVectorAndFloat(
     "Vector (RGB)"=&MaskVectorBySpawnGroup(
      Masked Vector=<10 5.3681 0.47>
      Passthrough Vector=&RandomRangeVector(Minimum=<5 1.0294 1.3281>, Maximum=<20 3.336 0>)
     )
     "Float (Alpha)"=&ClampFloat(Float=1)
    ))
   &AddVelocity(
    Rotation Coordinate Space=::World
    Velocity=&MaskVectorBySpawnGroup(Masked Vector=<0 0 50>, Passthrough Vector=<0 0 0>)
   ) {module=AddVelocity001}
  ]
  particleUpdateScript [
   &ParticleState()
   &ScaleSpriteSize(
    Scale Factor=&MaskVector2dBySpawnGroup(
     Masked Vector2D=&Vector2DFromCurve(
      Vector2Curve=DataInterface() {propertyValues {
        XCurve {keys [<at 0 1 RCIM_Cubic>, {interpMode=RCIM_Cubic, leaveTangent=-1.9252, time=1, value=0.2}]}
        YCurve {keys [<at 0 3 RCIM_Cubic>,<at 1 0 RCIM_Cubic>]}
       }}
      ScaleCurve=<1 1>
     )
     Passthrough Vector2D=&Vector2DFromFloat(Value=&FloatFromCurve(FloatCurve=DataInterface() {propertyValues {Curve {keys [<at 0 0>,<at 1 1>]}}}))
     Spawn Group Mask=0
    )
    Scale Sprite Size Mode=::NewEnumerator2
   )
   &AccelerationForce(Acceleration=&MaskVectorBySpawnGroup(Masked Vector=<0 0 -980>, Passthrough Vector=<0 0 0>))
   &Drag(
    Drag=&MaskFloatBySpawnGroup(Masked Float=3, Passthrough Float=1)
    Use Linear Drag=true
    Use Rotational Drag=true
   )
   &SolveForcesAndVelocity(Clamp Velocity=false, Limit Acceleration=false)
   &SubUVAnimation(Number Of Frames=&MaskFloatBySpawnGroup(Masked Float=0, Passthrough Float=3))
  ]
  rendererClasses [&NiagaraSpriteRendererProperties]
  renderers [
   {rendererClass=&NiagaraSpriteRendererProperties}
   {rendererClass=&NiagaraSpriteRendererProperties, rendererIndex=1}
  ]
 }
 Smoke {
  emitterUpdateScript [
   &EmitterState()
   &SpawnRate(
    SpawnRate=&Multiply_Float(
     A=20
     B=&MakeCustomFloatFromBool(
      Bool=$System.isRunning
      True Float=0
      False Float=&MakeCustomFloatFromBool(
       Bool=&SetBoolByFloatComparison(A=$System.Timeline002.Float, B=0, Comparison Type=::NewEnumerator4)
       True Float=0
       False Float=1
      )
     )
    )
    Use Spawn Probability=false
   )
  ]
  particleSpawnScript [&InitializeParticle(
    Color=<rgba 0.0782 0.0782 0.0782 0.0353>
    Sprite Rotation Mode=::NewEnumerator2
    Sprite Size Mode=::NewEnumerator3
    Uniform Sprite Size=5
    UsePositionOffset=false
    Write Material Random=true
   )]
  particleUpdateScript [
   &ParticleState()
   &SocketLocation(SamplePosOffset=false, Socket Reader=$System.Socket Reader, Socket Sampling Mode=::NewEnumerator3, VelocitySpeedLimit=false) {module=SocketLocation001}
   /Niagara/Modules/Update/SubUV/V2/SubUVAnimation(
    "Sprite Renderer "=DataInterface() {propertyValues {SpriteRenderer=&self:Smoke_0.NiagaraSpriteRendererProperties_0}}
    UseEndFrame=false
    UseStartFrame=false
   )
   &ShapeLocation(MaskSpawnGroup=false, Override Seed=false, Shape Origin=$Output.SocketLocation001.SampledPosition, Sphere Radius=1, UseOwnerScale=true)
   &ScaleColor(
    Scale Alpha=&Multiply_Float(
     A=$System.Timeline002.Float
     B=&FloatFromCurve(
      FloatCurve=DataInterface() {propertyValues {Curve {
         defaultValue=0
         keys [
          {arriveTangent=8.0089, interpMode=RCIM_Cubic, leaveTangent=8.0089, leaveTangentWeight=0.8048, tangentMode=RCTM_User, time=1.9, value=0}
          {tangentMode=RCTM_User, time=2, value=1}
         ]
        }}}
      CurveIndex=$Engine.System.Age
      Scale Curve=2.5
     )
    )
    ScaleA=true
    ScaleRGB=true
   )
  ]
  rendererClasses [&NiagaraSpriteRendererProperties]
  renderers [{rendererClass=&NiagaraSpriteRendererProperties}]
 }
 SmokeTrail {
  emitterUpdateScript [
   &EmitterState(Inactive Response=::NewEnumerator1, Life Cycle Mode=::NewEnumerator1, Loop Duration=1, UseLoopDelay=false)
   &SpawnRate(SpawnRate=60, Use Spawn Probability=false) {enabled=false, module=SpawnRate001}
   &SpawnBurst_Instantaneous(
    Spawn Group=-1
    Spawn Probability=HlslExpression() {hlslExpression="Emitter.LoopCount == 0"}
    Use Spawn Probability=true
    UseLoopCountLimit=false
   ) {enabled=false}
   &SpawnRate(
    SpawnRate=&Multiply_Float(
     A=80
     B=&MakeCustomFloatFromBool(
      Bool=$System.isRunning
      True Float=0
      False Float=&MakeCustomFloatFromBool(
       Bool=&SetBoolByFloatComparison(A=$System.Timeline002.Float, B=0, Comparison Type=::NewEnumerator4)
       True Float=0
       False Float=1
      )
     )
    )
    Use Spawn Probability=false
   )
  ]
  particleSpawnScript [
   &InitializeRibbon(Lifetime=0.2, Ribbon Width=4, UsePositionOffset=false, Write Material Random=true)
   &FN_FlipRibbonV()
   &SocketLocation(
    Interpolate Socket Sampling=true
    SamplePosOffset=false
    Socket Reader=$System.Socket Reader
    Socket Sampling Mode=::NewEnumerator3
    VelocitySpeedLimit=false
   )
   &AddVelocity(Rotation Coordinate Space=::World, Velocity=<0 0 100>)
   &AddVelocity(Velocity=<0 -40 0>) {module=AddVelocity001}
  ]
  particleUpdateScript [
   &InheritSourceMovement(
    Applied Position Scale=&Multiply_VectorByFloat(
     Vector=<1 1 1>
     Float=&FloatFromCurve(FloatCurve=DataInterface() {propertyValues {Curve {keys [<at 0 0.7>,<at 1 0.75>]}}})
    )
    ApplyForce=false
    ApplyPosition=true
    ApplyPositionFalloff=false
   )
   &ParticleState()
   &SocketLocation(
    Position Sampling=::NewEnumerator1
    SamplePosOffset=false
    Sampled Orientation Coordinate Space:ENiagaraCoordinateSpace=::Local
    Socket Reader=$System.Socket Reader
    Socket Sampling Mode=::NewEnumerator3
    VelocitySpeedLimit=false
   ) {module=SocketLocation001}
   &SolveForcesAndVelocity(Clamp Velocity=false, Limit Acceleration=false)
   /SetVariables(Particles.RenderPosition:Vector3f=&MakeCustomVectorFromBool(
     Bool=$Emitter.LocalSpace
     True Vector=&Add_Vector(
      A=Linked() {linkedVariable {name=Output.SocketLocation001.SampledPosition, type=NiagaraPosition}}
      B=Linked() {linkedVariable {name=Particles.Position, type=NiagaraPosition}}
     )
     False Vector=Linked() {linkedVariable {name=Particles.Position, type=NiagaraPosition}}
    )) {isSetParameters=true}
   &FN_RibbonMaterial(
    Color=&ColorFromCurve(ColorCurve=DataInterface() {propertyValues {
       AlphaCurve {keys [<at 0 1 RCIM_Cubic>,<at 1 0 RCIM_Cubic>]}
       BlueCurve {keys [<at 0 0.0521>,<at 1 0 RCIM_Cubic>]}
       GreenCurve {keys [<at 0 0.0521>,<at 1 0 RCIM_Cubic>]}
       RedCurve {keys [<at 0 0.0521>,<at 1 0 RCIM_Cubic>]}
      }})
    Color Pan U Speed=0.1
    Scale Alpha:NiagaraFloat=&Multiply_Float(
     A=&Multiply_Float(
      A=$System.Timeline002.Float
      B=&FloatFromCurve(
       FloatCurve=DataInterface() {propertyValues {Curve {
          defaultValue=0
          keys [
           {arriveTangent=8.0089, interpMode=RCIM_Cubic, leaveTangent=8.0089, leaveTangentWeight=0.8048, tangentMode=RCTM_User, time=1.9, value=0}
           {tangentMode=RCTM_User, time=2, value=1}
          ]
         }}}
       CurveIndex=$Engine.System.Age
       Scale Curve=2
      )
     )
     B=&FloatFromCurve(
      FloatCurve=DataInterface() {propertyValues {Curve {keys [{interpMode=RCIM_Cubic, tangentMode=RCTM_User, time=0, value=1}, {interpMode=RCIM_Cubic, tangentMode=RCTM_User, time=1, value=0}]}}}
      Scale Curve=&FNTimeOfDayScalar(Night Scale=1, Morning Scale=1, Evening Scale=1)
     )
    )
   ) {module=FN_RibbonMaterial001}
   &ScaleRibbonWidth(Ribbon Width Scale=&FloatFromCurve(FloatCurve=DataInterface() {propertyValues {Curve {keys [<at 0 0.5 RCIM_Cubic>,<at 1.2 2>]}}}, Scale Curve=2))
  ]
  rendererClasses [&NiagaraRibbonRendererProperties]
  renderers [{rendererClass=&NiagaraRibbonRendererProperties}]
 }
}
summary {
 systemName=NS_Backpack_TaperSlink_Idle
 emitters [
  {emitterName=RunningFire, rendererClasses [&NiagaraSpriteRendererProperties]}
  {emitterName=RunningSparks, rendererClasses [&NiagaraSpriteRendererProperties]}
  {emitterName=SmokeTrail, rendererClasses [&NiagaraRibbonRendererProperties]}
  {emitterName=IdleFire, rendererClasses [&NiagaraSpriteRendererProperties]}
  {emitterName=Smoke, rendererClasses [&NiagaraSpriteRendererProperties]}
  {emitterName=Light, rendererClasses [&NiagaraLightRendererProperties]}
 ]
 userVariables [
  User.bisFrontEnd {type=NiagaraBool}
  User.bIsFrontEndPreview {type=NiagaraBool}
 ]
}
